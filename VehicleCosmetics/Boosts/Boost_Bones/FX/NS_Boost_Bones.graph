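# Niagara car boost cosmetic that streams bones, glowing skulls and ribbon trails backward from the vehicle's boost sockets.
# Bones (bouncing under gravity) and rising Skull sprites spread across filtered skeletal-mesh sockets, alongside quality-scaled Trail ribbons.
# Skull emits location events; the TrailSkull ribbon emitter has zero spawn rate of its own.
# Each emitter applies painted-variant colors and inherits source movement, with frontend-preview switches adjusting velocities and some lifetimes.
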NiagaraSystem /VehicleCosmetics/Boosts/Boost_Bones/FX/NS_Boost_Bones
dependencies {
 usedDataInterfaces [
  &NiagaraDataInterfaceCurve
  &NiagaraDataInterfaceSpriteRendererInfo
  &NiagaraDataInterfaceSkeletalMesh
 ]
 usedDynamicInputs []
 usedModules [
  &self:SystemSpawnScript.SystemScriptSource.SystemScriptGraph.NiagaraNodeAssignment_0.SetVariables_#1
  &SystemState
 ]
}
emitters {
 Bones {
  emitterUpdateScript [
   &EmitterState(Inactive Response=::NewEnumerator1, Life Cycle Mode=::NewEnumerator1, Loop Duration=1, UseLoopDelay=false)
   &VC_FilteredSocketCount()
   &SpawnRate(Spawn Probability=0.7, SpawnRate=20, Use Spawn Probability=true)
   &SpawnRate(Spawn Probability=0.7, SpawnGroup=1, SpawnRate=20, Use Spawn Probability=true) {module=SpawnRate001}
  ]
  particleSpawnScript [
   &VC_PaintedVariantsOverride(
    Black=<rgba 0 0 0 2>
    BurntSienna=<rgba 0 0 0 1>
    Cobalt=<rgba 0.1146 0.131 1 1>
    Crimson=<rgba 1 0.026 0.0524 1>
    DefaultColor=<rgba 0.6333 1 0.6 1>
    ForestGreen=<rgba 0.474 1 0.5005 1>
    Grey=<rgba 0 0 0 1>
    Lime=<rgba 0.5739 0.776 0.3719 1>
    Orange=<rgba 0.349 0.1691 0.0727 1>
    Pink=<rgba 0.5052 0.1947 0.3923 1>
    Purple=<rgba 0.1788 0.1003 0.2917 1>
    R1 Black:LinearColor=<rgba 0.01 0.01 0.01 1>
    R1 BurntSienna:LinearColor=<rgba 0.01 0.01 0.01 1>
    R1 Cobalt:LinearColor=<rgba 0.01 0.01 0.01 1>
    R1 Crimson:LinearColor=<rgba 0.01 0.01 0.01 1>
    R1 DefaultColor=<rgba 0.01 0.01 0.01 1>
    R1 ForestGreen:LinearColor=<rgba 0.01 0.01 0.01 1>
    R1 Gold:LinearColor=<rgba 0.01 0.01 0.01 1>
    R1 Grey:LinearColor=<rgba 0.01 0.01 0.01 1>
    R1 Lime:LinearColor=<rgba 0.01 0.01 0.01 1>
    R1 Orange:LinearColor=<rgba 0.01 0.01 0.01 1>
    R1 Pink:LinearColor=<rgba 0.01 0.01 0.01 1>
    R1 Purple:LinearColor=<rgba 0.01 0.01 0.01 1>
    R1 Saffron:LinearColor=<rgba 0.01 0.01 0.01 1>
    R1 SkyBlue:LinearColor=<rgba 0.01 0.01 0.01 1>
    R1 TitaniumWhite:LinearColor=<rgba 0.01 0.01 0.01 1>
    Saffron=<rgba 0.9 0.7556 0.2812 1>
    SkyBlue=<rgba 0.2626 0.4448 0.5365 1>
    Use R1 Painted Overrides=true
   )
   &InitializeParticle(
    Color=$Particles.Color
    Lifetime=&MakeCustomFloatFromBool(
     Bool=$User.bIsFrontendPreview
     True Float=&RandomRangeFloat(Minimum=0.7, Maximum=0.8)
     False Float=&RandomRangeFloat(Minimum=0.6, Maximum=0.8)
    )
    Sprite Rotation Mode=::NewEnumerator2
    Sprite Size Mode=::NewEnumerator1
    Uniform Sprite Size Max=30
    Uniform Sprite Size Min=40
    UsePositionOffset=false
    Write Material Random=true
   )
   &PartitionParticles(Number Of Partitions=$Emitter.PortCount)
   &SkeletalMeshLocation(
    Bone / Socket Index=$Particles.PartitionParticles.Partition
    Bone / Socket Sampling Mode=::NewEnumerator5
    MaskSpawnGroup=false
    Mesh Sampling Type=::NewEnumerator6
    Skeletal Mesh=$System.SkeletalMesh
   )
   &ShapeLocation(
    MaskSpawnGroup=false
    Offset=&AddVectorToPosition(Position=<-20 0 0>, Vector=$User.DTCarOffset)
    Offset Coordinate Space=::Local
    Offset Mode=::NewEnumerator0
    Override Seed=false
    Sphere Radius=40
    UseOwnerScale=true
   )
   &AddVelocity(Velocity=&MakeCustomVectorFromBool(
     Bool=$User.bIsFrontendPreview
     True Vector=&RandomRangeVector(Minimum=<-1100 -120 100>, Maximum=<-1300 120 400>)
     False Vector=&RandomRangeVector(Minimum=<-1100 -120 100>, Maximum=<-1300 120 400>)
    ))
   /SetVariables(Particles.VisibilityTag=$Emitter.SpawnGroup) {isSetParameters=true}
  ]
  particleUpdateScript [
   &ParticleState()
   &SpriteRotationRate(Rotation Rate=&RandomRangeFloat(Minimum=-200, Maximum=200))
   &ScaleSpriteSize(
    Uniform Curve Scale=1.8
    Uniform Curve Sprite Scale=DataInterface() {propertyValues {Curve {keys [{tangentMode=RCTM_User, time=0, value=0}, <at 0.1 1>, {tangentMode=RCTM_User, time=1, value=1}]}}}
   )
   &GravityForce(Gravity=<0 0 -900>)
   &Collision(
    Advanced Aging Rate=1
    Age Colliding Particles=false
    Analytical Collision Plane Position 1:Vector3f=<0 0 1>
    Analytical Collision Plane Position 2:Vector3f=<0 0 -5>
    CPU Collision Type=::NewEnumerator3
    Friction=1
    KillOnCollision=false
    Maximum Penetration Correction Distance=0.3
    Module.Control Rest State Via Surface Normal=false
    Module.EnableMaxCollisionCount=false
    Module.Randomize Collision Normal Vector=false
    Restitution=0.8
    Simple Friction=false
   )
   &InheritSourceMovement(ApplyForce=false, ApplyPosition=true, ApplyPositionFalloff=false)
   &SolveForcesAndVelocity(Clamp Velocity=false, Limit Acceleration=false)
   &SubUVAnimation(
    "Sprite Renderer "=DataInterface() {propertyValues {SpriteRenderer=&self:Bones_0.NiagaraSpriteRendererProperties_3}}
    SubUV Lookup Index Scale=2
    UseEndFrame=false
    UseStartFrame=false
   )
   &ScaleColor(Scale RGB=&VectorFromFloat(Value=4), ScaleA=false, ScaleRGB=true)
  ]
  rendererClasses [&NiagaraSpriteRendererProperties]
  renderers [
   {rendererClass=&NiagaraSpriteRendererProperties}
   {rendererClass=&NiagaraSpriteRendererProperties, rendererIndex=1}
   {rendererClass=&NiagaraSpriteRendererProperties, rendererIndex=2}
   {rendererClass=&NiagaraSpriteRendererProperties, rendererIndex=3}
  ]
 }
 Skull {
  emitterUpdateScript [
   &EmitterState(Inactive Response=::NewEnumerator1, Life Cycle Mode=::NewEnumerator1, Loop Duration=1, UseLoopDelay=false)
   &VC_FilteredSocketCount()
   &SpawnRate(Spawn Probability=0.7, SpawnRate=12, Use Spawn Probability=true)
  ]
  particleSpawnScript [
   &VC_PaintedVariantsOverride(
    Black=<rgba 0 0 0 2>
    BurntSienna=<rgba 0.0365 0.0077 0 1.5>
    Cobalt=<rgba 0.0088 0.0164 0.4219 2>
    Crimson=<rgba 0.1771 0.0101 0.0147 1.5>
    DefaultColor=<rgba 0.3 0.95 0.2 1>
    ForestGreen=<rgba 0 0.4696 0.0237 2>
    Grey=<rgba 0.0104 0.0104 0.0104 2>
    Lime=<rgba 0.1068 0.2135 0 2.5>
    Orange=<rgba 0.2031 0.0709 0 3>
    Pink=<rgba 0.151 0.0181 0.1027 5>
    Purple=<rgba 0.0612 0.0055 0.224 5>
    Saffron=<rgba 0.1927 0.1477 0 3>
    SkyBlue=<rgba 0.0216 0.1528 0.2188 2.5>
    TitaniumWhite=<rgba 0.4896 0.4896 0.4896 1>
   )
   &InitializeParticle(
    Color=$Particles.Color
    Lifetime=&MakeCustomFloatFromBool(
     Bool=$User.bIsFrontendPreview
     True Float=&RandomRangeFloat(Minimum=1, Maximum=1.2)
     False Float=&RandomRangeFloat(Minimum=0.8, Maximum=1)
    )
    Sprite Rotation Angle=&RandomRangeFloat(Minimum=-25, Maximum=25)
    Sprite Rotation Mode=::NewEnumerator0
    Sprite Size Mode=::NewEnumerator1
    Uniform Sprite Size Max=90
    Uniform Sprite Size Min=70
    UsePositionOffset=false
    Write Material Random=true
   )
   &PartitionParticles(Number Of Partitions=$Emitter.PortCount)
   &SkeletalMeshLocation(
    Bone / Socket Index=$Particles.PartitionParticles.Partition
    Bone / Socket Sampling Mode=::NewEnumerator5
    MaskSpawnGroup=false
    Mesh Sampling Type=::NewEnumerator6
    Skeletal Mesh=$System.SkeletalMesh
   )
   &ShapeLocation(
    MaskSpawnGroup=false
    Offset=&AddVectorToPosition(Position=<-20 0 0>, Vector=$User.DTCarOffset)
    Offset Coordinate Space=::Local
    Offset Mode=::NewEnumerator0
    Override Seed=false
    Sphere Radius=20
    UseOwnerScale=true
   )
   &AddVelocity(Velocity=&MakeCustomVectorFromBool(
     Bool=$User.bIsFrontendPreview
     True Vector=&RandomRangeVector(Minimum=<-500 -200 50>, Maximum=<-600 200 100>)
     False Vector=&RandomRangeVector(Minimum=<-1000 -130 0>, Maximum=<-1200 130 70>)
    ))
  ]
  particleUpdateScript [
   &ParticleState()
   &ScaleSpriteSize(Uniform Curve Sprite Scale=DataInterface() {propertyValues {Curve {keys [
        {arriveTangent=1.4129, interpMode=RCIM_Cubic, leaveTangent=1.4129, leaveTangentWeight=0.0667, tangentMode=RCTM_User, time=0, value=0.6}
        {arriveTangent=1.2267, arriveTangentWeight=0.0667, interpMode=RCIM_Cubic, leaveTangent=1.2267, tangentMode=RCTM_User, time=0.2, value=1.3}
       ]}}})
   &AccelerationForce(Acceleration=&RandomRangeVector(Minimum=<1 1 60>, Maximum=<1 1 120>))
   &InheritSourceMovement(ApplyForce=false, ApplyPosition=true, ApplyPositionFalloff=false)
   &SolveForcesAndVelocity(Clamp Velocity=false, Limit Acceleration=false)
   &SubUVAnimation(
    "Sprite Renderer "=DataInterface() {propertyValues {SpriteRenderer=&self:Skull_0.NiagaraSpriteRendererProperties_0}}
    SubUV Lookup Index Scale=3
    UseEndFrame=false
    UseStartFrame=false
   )
   &ScaleColor(
    Scale Alpha=&FloatFromCurve(FloatCurve=DataInterface() {propertyValues {Curve {keys [<at 0 0>,<at 0.2 0.9>,<at 1 0>]}}})
    Scale RGB=&VectorFromFloat(Value=3)
    ScaleA=true
    ScaleRGB=true
   )
   /SetVariables(Particles.eyescolor:LinearColor=&MultiplyLinearColor(
     Linear Color=$Particles.Color
     Scale Factor=&MakeVector4FromVectorAndFloat(
      "Vector (RGB)"=&VectorFromFloat(Value=3)
      "Float (Alpha)"=&FloatFromCurve(FloatCurve=DataInterface() {propertyValues {Curve {keys [
           <at 0.3 1 RCIM_Cubic>
           {arriveTangent=-0.4556, arriveTangentWeight=0.1871, interpMode=RCIM_Cubic, leaveTangent=-0.4556, tangentMode=RCTM_User, time=0.8, value=0}
          ]}}})
     )
    )) {isSetParameters=true}
   &GenerateLocationEvent(Event Send Rate=20, Use Event Probability=false, UseEventDelay=false)
  ]
  rendererClasses [&NiagaraSpriteRendererProperties]
  renderers [
   {rendererClass=&NiagaraSpriteRendererProperties}
   {rendererClass=&NiagaraSpriteRendererProperties, rendererIndex=1}
  ]
 }
 Trail {
  emitterUpdateScript [
   &EmitterState(Inactive Response=::NewEnumerator1, Life Cycle Mode=::NewEnumerator1, Loop Duration=1, UseLoopDelay=false)
   &VC_FilteredSocketCount()
   &VC_RateQualityHandler(Epic Spawn Rate=50, High Spawn Rate=40, Low Spawn Rate=25, Use Custom Rate?=true)
   &VC_SpawnRibbon_v1(Number of Ribbons=$Emitter.PortCount, SpawnRate=$Emitter.Spawn Rate)
  ]
  particleSpawnScript [
   &VC_PaintedVariantsOverride(
    Black=<rgba 0 0 0 2>
    BurntSienna=<rgba 0.1667 0.0353 0 0.5304>
    Cobalt=<rgba 0.0138 0.0298 0.8802 0.6261>
    Crimson=<rgba 0.4635 0.0338 0.0454 0.2783>
    DefaultColor=<rgba 0.1225 0.7 0.07 1>
    ForestGreen=<rgba 0.0217 0.2083 0.0263 0.487>
    Grey=<rgba 0.0469 0.0469 0.0469 1>
    Lime=<rgba 0.1716 0.3021 0 0.4348>
    Orange=<rgba 0.3043 0.1062 0 0.7391>
    Pink=<rgba 0.5677 0.0266 0.3709 0.2348>
    Purple=<rgba 0.2789 0 0.6795 0.4174>
    Saffron=<rgba 0.2604 0.2316 0 0.6087>
    SkyBlue=<rgba 0.0574 0.338 0.4792 0.3391>
    TitaniumWhite=<rgba 0.75 0.75 0.75 0.2957>
   )
   &VC_InitializeRibbon_v1(
    Color=$Particles.Color
    Custom Offset=true
    Enable Skeletal Position=true
    Lifetime=1.3
    Local Vector Offset=&AddVectorToPosition(Position=<0 0 0>, Vector=$User.DTCarOffset)
    Offset Amount=1
    Ribbon Width=100
   )
   &AddVelocity(Velocity=&MakeCustomVectorFromBool(
     Bool=$User.bIsFrontendPreview
     True Vector=&RandomRangeVector(Minimum=<-1300 -20 -20>, Maximum=<-1300 20 20>)
     False Vector=&RandomRangeVector(Minimum=<-1200 -20 -20>, Maximum=<-1200 20 20>)
    )) {module=AddVelocity001}
  ]
  particleUpdateScript [
   &ParticleState()
   &InheritSourceMovement(ApplyForce=false, ApplyPosition=true, ApplyPositionFalloff=false)
   &SolveForcesAndVelocity(Clamp Velocity=false, Limit Acceleration=false)
   &VC_UpdateRibbon_v1(Enable Skeletal Position=true)
   &ScaleRibbonWidth(Ribbon Width Scale=&FloatFromCurve(FloatCurve=DataInterface() {propertyValues {Curve {keys [
         {interpMode=RCIM_Cubic, leaveTangentWeight=0.3355, tangentMode=RCTM_Break, time=0, value=1}
         {arriveTangent=-1.2599, interpMode=RCIM_Cubic, tangentMode=RCTM_Break, time=1, value=0}
        ]}}}))
   &ScaleColor(
    Scale Alpha=&FloatFromCurve(FloatCurve=DataInterface() {propertyValues {Curve {keys [
         {interpMode=RCIM_Cubic, leaveTangentWeight=0.15, tangentMode=RCTM_User, time=0, value=0}
         {arriveTangentWeight=0.15, interpMode=RCIM_Cubic, leaveTangentWeight=0.15, tangentMode=RCTM_User, time=0.45, value=0.6}
         {arriveTangentWeight=0.15, interpMode=RCIM_Cubic, tangentMode=RCTM_User, time=0.9, value=0}
        ]}}})
    Scale RGB=&VectorFromFloat(Value=&FloatFromCurve(FloatCurve=DataInterface() {propertyValues {Curve {keys [
          {interpMode=RCIM_Cubic, leaveTangentWeight=0.1, tangentMode=RCTM_User, time=0.1, value=1}
          {arriveTangentWeight=0.1, interpMode=RCIM_Cubic, leaveTangentWeight=0.2333, tangentMode=RCTM_User, time=0.5, value=3}
          {arriveTangentWeight=0.2333, interpMode=RCIM_Cubic, tangentMode=RCTM_User, time=1, value=1}
         ]}}}))
    ScaleA=true
    ScaleRGB=true
   )
  ]
  rendererClasses [&NiagaraRibbonRendererProperties]
  renderers [{rendererClass=&NiagaraRibbonRendererProperties}]
 }
 TrailSkull {
  emitterUpdateScript [
   &EmitterState(Inactive Response=::NewEnumerator1, Life Cycle Mode=::NewEnumerator1, Loop Duration=0.3, UseLoopDelay=false)
   &VC_FilteredSocketCount()
   &VC_RateQualityHandler(Epic Spawn Rate=40, High Spawn Rate=35, Low Spawn Rate=30, Use Custom Rate?=true)
   &VC_SpawnRibbon_v1(Number of Ribbons=$Emitter.PortCount, SpawnRate=0)
  ]
  particleSpawnScript [
   &VC_PaintedVariantsOverride(
    Black=<rgba 0 0 0 0.3304>
    BurntSienna=<rgba 0.1771 0.0375 0 0.1478>
    Cobalt=<rgba 0.0677 0.085 1 0.2957>
    Crimson=<rgba 0.6198 0.0291 0.0451 0.25>
    DefaultColor=<rgba 0.16 0.4 0.16 0.4>
    ForestGreen=<rgba 0 0.5 0.0253 0.4609>
    Grey=<rgba 0.1094 0.1094 0.1094 0.4087>
    Lime=<rgba 0.1823 0.3646 0 0.3217>
    Orange=<rgba 0.6724 0.2346 0 0.2783>
    Pink=<rgba 0.5052 0.1342 0.3703 0.3478>
    Purple=<rgba 0.2956 0.0283 0.6795 0.2174>
    Saffron=<rgba 0.9 0.69 0 0.1043>
    SkyBlue=<rgba 0.0391 0.37 0.5365 0.113>
    TitaniumWhite=<rgba 0.75 0.75 0.75 0.0522>
   )
   &VC_InitializeRibbon_v1(Color=$Particles.Color, Enable Skeletal Position=true, Lifetime=0.7, Offset Amount=1, Ribbon Width=100)
   &AddVelocity(Velocity=&MakeCustomVectorFromBool(Bool=$User.bIsFrontendPreview, True Vector=<800 0 0>, False Vector=<100 0 0>)) {module=AddVelocity001}
  ]
  particleUpdateScript [
   &ParticleState()
   &InheritSourceMovement(ApplyForce=false, ApplyPosition=true, ApplyPositionFalloff=false)
   &SolveForcesAndVelocity(Clamp Velocity=false, Limit Acceleration=false)
   &VC_UpdateRibbon_v1(Enable Skeletal Position=true)
   &ScaleRibbonWidth(Ribbon Width Scale=&FloatFromCurve(FloatCurve=DataInterface() {propertyValues {Curve {keys [<at 0 1.2>,<at 0.9 0>]}}}))
   &ScaleColor(
    Scale Alpha=&FloatFromCurve(FloatCurve=DataInterface() {propertyValues {Curve {keys [
         {interpMode=RCIM_Cubic, leaveTangent=-1.7164, tangentMode=RCTM_Break, time=0, value=1}
         {interpMode=RCIM_Cubic, tangentMode=RCTM_Break, time=0.8, value=0}
        ]}}})
    ScaleA=true
    ScaleRGB=false
   )
  ]
  rendererClasses [&NiagaraRibbonRendererProperties]
  renderers [{rendererClass=&NiagaraRibbonRendererProperties}]
 }
}
summary {
 systemName=NS_Boost_Bones
 emitters [
  {emitterName=Bones, rendererClasses [&NiagaraSpriteRendererProperties]}
  {emitterName=Skull, rendererClasses [&NiagaraSpriteRendererProperties]}
  {emitterName=Trail, rendererClasses [&NiagaraRibbonRendererProperties]}
  {emitterName=TrailSkull, rendererClasses [&NiagaraRibbonRendererProperties]}
 ]
 userVariables [
  User.bIsFrontendPreview {type=NiagaraBool}
  User.bIsPainted {type=NiagaraBool}
  User.ColorVariantID {type=NiagaraInt32}
  User.DTCarOffset {type=Vector3f, description="", defaultValue {struct=Vector3f, value=<0 0 0>}}
  User.PaintedColor {type=LinearColor, description="", defaultValue {struct=LinearColor, value=<rgba 1 1 1 1>}}
  User.PaintedScalar {type=NiagaraFloat, defaultValue=1, description=""}
  User.PortCount {type=NiagaraInt32, defaultValue=2}
 ]
}
