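# Unlit masked Niagara sprite material that draws flipbook bones for the Bone boost vehicle cosmetic.
# A TextureSwitch dynamic parameter, snapped with a threshold, selects between two SubUV bone textures for both RGB and alpha.
# The RGB acts as normals for fake lighting: camera and up-axis shading, an environment lookup and a Fresnel rim.
# Particle color, monochromatically shifted and desaturated, tints that lighting before an emissive adjustment function.
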Material /VehicleCosmetics/Boosts/Boost_Bones/Materials/M_BoostBones_Bones
graph {
 properties {bUsedWithNiagaraSprites=true, blendMode=BLEND_Masked, pixelDepthOffsetMode=PDOM_Legacy, shadingModel=MSM_Unlit, usedShadingModels=Unlit}
 nodes [
  n1 = NamedRerouteUsage()
  MaterialExpressionVertexColor_0 = ParticleColor()
  n2 = CameraVectorWS()
  n3 = TextureSampleParameterSubUV!(UVs, Apply View MipBias) {parameterName=Bone_01, samplerType=SAMPLERTYPE_Masks, texture=/DealershipSharedLibrary/FX/Textures/Flipbook/T_VCVFX_BoostBones_Bone_01}
  n4 = TextureSampleParameterSubUV() {parameterName=Bone_02, samplerType=SAMPLERTYPE_Masks, texture=/DealershipSharedLibrary/FX/Textures/Flipbook/T_VCVFX_BoostBone_02}
  n5 = DynamicParameter() {paramNames [TextureSwitch, Param2, Param3, Param4]}
  n6 = Ceil(None=@n5.TextureSwitch)
  n7 = If(A=@n5.TextureSwitch, B=Constant() {r=0.49}, "A > B"=@n6, "A == B"=@n6, "A < B"=Floor(None=@n5.TextureSwitch))
  n8 = LinearInterpolate(A=@n3.RGBA, B=@n4.RGBA, Alpha=@n7)
  MaterialExpressionConstantClamp_0 = Saturate(None=OneMinus(None=DotProduct(A=@n2, B=@n8)))
  n9 = ConstantBiasScale(None=DotProduct(A=Transform(None=@n8) {transformSourceType=TRANSFORMSOURCE_World}, B=Constant3Vector() {constant=<rgba 0 0 1 0>})) {bias=0.6, scale=1}
  n10 = Constant() {r=1}
  n11 = Constant()
  n12 = Add(
   A=Multiply(A=@MaterialExpressionVertexColor_0.A, B=Multiply(A=@MaterialExpressionConstantClamp_0, B=@n9))
   B=Multiply(
    A=MaterialFunctionCall!(Desaturate=@n10, "Tint (Curve)"=@n11, Tint Amount=@n11, Cubemap) {materialFunction=/DealershipSharedLibrary/FX/Materials/MaterialFunctions/VC_MF_WorldEnv}.LocalSpace
    B=Constant() {r=1}
   )
  )
  n13 = Transform(None=CameraVectorWS()) {transformSourceType=TRANSFORMSOURCE_World, transformType=TRANSFORM_Tangent}
  n14 = Constant() {r=2}
  n15 = Multiply(
   A=Multiply(
    A=MaterialFunctionCall(Normal Vector=@n8, Camera Vector=@n13, Power=@n14) {materialFunction=&Fresnel_Function}.Result
    B=Constant() {r=2}
   ) {constB=2}
   B=@MaterialExpressionVertexColor_0.A
  )
  n16 = Constant() {r=0.75}
  n17 = Multiply(A=@MaterialExpressionVertexColor_0.RGBA, B=Constant() {r=1.5})
  n18 = Desaturation(
   None=Multiply(
    A=MaterialFunctionCall!(Shift=@n16, Color=@n17) {materialFunction=&/MF_MonochromaticShift}.Result
    B=Constant() {r=1.25}
   )
   Fraction=Constant() {r=0.6}
  )
  MaterialFunctionCall(Emissive=Add(A=Multiply(A=@n1, B=@n12), B=Multiply(A=@n15, B=@n18))) {materialFunction=&MF_FX_EmissiveAdjustments}
  NamedRerouteDeclaration(None=Multiply(A=@n18, B=SquareRoot(None=@n18))) {name=MonoChromatic, nodeColor=<rgba 0 1 0.8359 1>}
  Reroute(None=LinearInterpolate(A=@n3.A, B=@n4.A, Alpha=@n7))
 ]
}
