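# Translucent unlit Niagara sprite material for ghostly skulls in the Bones vehicle boost effect.
# Time steps through a 4x4 skull flipbook and matching eye-glow sheet; the skull's RGB serves as a normal for fake Fresnel rim light.
# IsGlitchy enables a scrolling glitch texture that distorts UVs, dissolves opacity and modulates emission.
# Dynamic parameters drive animation speed, brightness, alpha and glitch per particle, alongside particle color.
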Material /VehicleCosmetics/Boosts/Boost_Bones/Materials/M_BoostBones_GhostlySkulls
graph {
 properties {
  bUsedWithNiagaraSprites=true
  blendMode=BLEND_Translucent
  pixelDepthOffsetMode=PDOM_Legacy
  shadingModel=MSM_Unlit
  twoSided=true
  usedShadingModels=Unlit
 }
 nodes [
  n1 = ScalarParameter() {parameterName="IsGlitchy (0 False 1 True)"}
  n2 = SquareRoot(None=NamedRerouteUsage())
  n3 = Multiply(A=@n2, B=NamedRerouteUsage())
  n4 = Constant() {r=2}
  n5 = DynamicParameter() {
   parameterIndex=1
   defaultValue=<rgba -1 1 1 0.2>
   paramNames [Glitch - Alpha Dissolve, Glitch - Tile X, Glitch - Emission, Glitch - Distortion]
  }
  n6 = DynamicParameter() {parameterIndex=2, defaultValue=<rgba 5 0 0.2 0.2>, paramNames [Glitch - Tile Y, Glitch - Scroll Speed X, Glitch - Scroll Speed Y, Param4]}
  n7 = AppendVector(A=@n5.Glitch - Tile X, B=@n6.Glitch - Tile Y)
  n8 = TextureCoordinate()
  n9 = Time()
  n10 = TextureSample(UVs=Panner(Coordinate=Multiply(A=@n7, B=@n8), Time=@n9, Speed=AppendVector(A=@n6.Glitch - Scroll Speed X, B=@n6.Glitch - Scroll Speed X)) {speedX=0.1, speedY=0.025}) {texture=/DealershipSharedLibrary/FX/Textures/Digital/T_DLR_Glitch_FB_RGB_03_T}
  n11 = ComponentMask(None=@n10.RGB) {g=true, r=true}
  n12 = Constant() {r=0.5}
  n13 = ScalarParameter() {defaultValue=4}
  n14 = ScalarParameter() {defaultValue=4}
  n15 = AppendVector(A=@n13, B=@n14)
  n16 = DynamicParameter() {defaultValue=<rgba 1 1 1 0.2>, paramNames [AnimationSpeed, Brightness, FresnelBrightness, Alpha_Skull]}
  n17 = Frac(None=Multiply(A=@n9, B=@n16.AnimationSpeed))
  n18 = Multiply(A=AppendVector(A=Multiply(A=@n13, B=@n14), B=@n14), B=AppendVector(A=@n17, B=@n17))
  n19 = Add(A=Divide(A=Panner(Coordinate=@n8, Time=@n12) {speedX=0.1, speedY=0.025}, B=@n15), B=Divide(A=Floor(None=@n18), B=@n15))
  n20 = If(
   A=ScalarParameter() {parameterName="IsGlitchy (0 False 1 True)"}
   "A > B"=Add(A=Multiply(A=@n11, B=@n5.Glitch - Distortion), B=@n19)
   "A == B"=@n19
   "A < B"=@n19
  )
  n21 = Multiply(
   A=Multiply(
    A=Multiply(A=@n3, B=@n4)
    B=TextureSample(UVs=@n20) {texture=/DealershipSharedLibrary/FX/Textures/Flipbook/T_VCVFX_BoostBones_Skull_EyeGlow}.RGBA
   )
   B=NamedRerouteUsage()
  )
  n22 = TextureSample(UVs=@n20) {samplerType=SAMPLERTYPE_Masks, texture=/DealershipSharedLibrary/FX/Textures/Flipbook/T_VCVFX_BoostBones_ScarySkull_FB}
  n23 = Add(A=@n21, B=Multiply(A=@n22.A, B=@n16.Alpha_Skull))
  n24 = Desaturation(None=NamedRerouteUsage())
  n25 = DynamicParameter() {
   parameterIndex=1
   defaultValue=<rgba -1 1 1 0.2>
   paramNames [Glitch - Alpha Dissolve, Glitch - Tile X, Glitch - Emission, Glitch - Distortion]
  }
  n26 = Multiply(A=Saturate(None=@n23), B=Saturate(None=Subtract(A=@n24, B=@n25.Glitch - Alpha Dissolve)))
  MaterialExpressionVertexColor_5 = ParticleColor()
  MaterialExpressionConstantClamp_7 = Saturate(None=Multiply(A=If(A=@n1, "A > B"=@n26, "A == B"=@n23, "A < B"=@n23), B=@MaterialExpressionVertexColor_5.A))
  n27 = NamedRerouteUsage()
  n28 = MaterialFunctionCall(
   Normal Vector=MaterialFunctionCall(X=Power(Base=@n22.R), Y=Power(Base=@n22.G), Z=@n22.B) {materialFunction=&MakeFloat3}.Result
   Camera Vector=Transform(None=CameraVectorWS()) {transformSourceType=TRANSFORMSOURCE_World, transformType=TRANSFORM_Tangent}
   Power=Constant() {r=2.25}
  ) {materialFunction=&Fresnel_Function}
  n29 = Constant() {r=2}
  n30 = Multiply(A=Multiply(A=@n2, B=NamedRerouteUsage()), B=Multiply(A=@n16.FresnelBrightness, B=Multiply(A=@n28.Result, B=@n29) {constB=2}))
  n31 = Add(A=Multiply(A=NamedRerouteUsage(), B=NamedRerouteUsage()), B=@n30)
  n32 = Add(A=Multiply(A=@n30, B=ScalarParameter() {defaultValue=9, parameterName=SkullFresnel - Intensity}), B=@n31)
  n33 = LinearInterpolate(B=Multiply(A=Multiply(A=@n27, B=@n5.Glitch - Emission), B=@n32), Alpha=@n32)
  n34 = Add(A=@n21, B=@n31)
  MaterialFunctionCall(Emissive=LinearInterpolate(A=@n34, B=Multiply(A=@n33, B=@n34), Alpha=ScalarParameter() {parameterName="IsGlitchy (0 False 1 True)"})) {materialFunction=&MF_FX_EmissiveAdjustments}
  NamedRerouteDeclaration(None=@MaterialExpressionVertexColor_5.RGBA) {name=RGBA, nodeColor=<rgba 1 0 0.4922 1>}
  NamedRerouteDeclaration(None=@n16.Brightness) {name=Brightness, nodeColor=<rgba 1 0.8672 0 1>}
  NamedRerouteDeclaration(None=Desaturation(None=@n22.RGBA)) {name=Desaturation, nodeColor=<rgba 0.2676 0.3285 1 1>}
  NamedRerouteDeclaration(None=@n10.RGBA) {name=GlitchText, nodeColor=<rgba 0.4766 0 1 1>}
  ScalarParameter() {parameterName=Glitch - X Scroll Speed}
  ScalarParameter() {defaultValue=-0.0661, parameterName=Distortion}
  ScalarParameter() {defaultValue=0.2, parameterName=Glitch - Y Scroll Speed}
  ScalarParameter() {defaultValue=5, parameterName=Glitch - Y Tile}
 ]
}
