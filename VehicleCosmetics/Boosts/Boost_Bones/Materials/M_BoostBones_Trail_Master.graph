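# Translucent unlit master material for Niagara ribbon trails on the Bones boost vehicle cosmetic.
# Two panning smoke textures, distorted by flow noise and scaled by PowerTrail, add to a centre band shaped from V.
# A U-based fade, switched by SkullGhostTrail between a cone-gradient mask and a brighter unclamped ramp, multiplies this with particle alpha into a saturated mask.
# Particle color feeds the shared MF_FX_EmissiveAdjustments function.
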Material /VehicleCosmetics/Boosts/Boost_Bones/Materials/M_BoostBones_Trail_Master
graph {
 properties {bUsedWithNiagaraRibbons=true, blendMode=BLEND_Translucent, pixelDepthOffsetMode=PDOM_Legacy, shadingModel=MSM_Unlit, usedShadingModels=Unlit}
 nodes [
  MaterialExpressionVertexColor_2 = ParticleColor()
  n1 = Constant() {r=0.8}
  n2 = TextureSample(UVs=Rotator(Time=Constant() {r=0.5}) {speed=3.14}) {texture=/DealershipSharedLibrary/FX/Textures/Gradient_Masks/T_VCVFX_ConeGradient_A}
  n3 = TextureCoordinate()
  n4 = Multiply(
   A=StaticSwitchParameter(True=Constant() {r=2}, False=Multiply(A=@n1, B=@n2.RGBA)) {parameterName=SkullGhostTrail?}
   B=ComponentMask(None=@n3) {r=true}
  )
  MaterialExpressionConstantClamp_4 = Saturate(None=@n4)
  n5 = StaticSwitchParameter(True=@n4, False=@MaterialExpressionConstantClamp_4) {parameterName=SkullGhostTrail?}
  n6 = ComponentMask(None=@n3) {g=true}
  n7 = Multiply(A=@n6, B=OneMinus(None=@n6))
  n8 = Constant() {r=4}
  n9 = TextureCoordinate() {uTiling=0.5, vTiling=0.5}
  n10 = ComponentMask(None=TextureSample(UVs=Panner(Coordinate=@n9) {speedX=-0.5, speedY=0.25}) {texture=/DealershipSharedLibrary/FX/Textures/Noise/T_VCVFX_Flow03_N}.RGBA) {g=true, r=true}
  n11 = Constant() {r=0.25}
  n12 = Add(
   A=Panner(Coordinate=TextureCoordinate() {uTiling=0, vTiling=0}, Time=Time()) {speedX=-1}
   B=Add(A=TextureCoordinate() {uTiling=0.25}, B=Multiply(A=@n10, B=@n11))
  )
  n13 = TextureSample(UVs=@n12) {texture=/DealershipSharedLibrary/FX/Textures/Gradient_Masks/T_VCVFX_Smoke_Trail_02}
  n14 = TextureSample(UVs=@n12) {texture=/DealershipSharedLibrary/FX/Textures/TexturePackage/T_VCVFX_SmokeTrail_RGB}
  n15 = Add(
   A=Multiply(A=Constant() {r=25}, B=Power(Base=@n7, Exp=@n8))
   B=Multiply(A=Add(A=@n13.R, B=@n14.B), B=ScalarParameter() {defaultValue=8, parameterName=PowerTrail})
  )
  MaterialExpressionConstantClamp_3 = Saturate(None=Multiply(A=@MaterialExpressionVertexColor_2.A, B=Multiply(A=@n5, B=@n15)))
  MaterialFunctionCall(Emissive=@MaterialExpressionVertexColor_2.RGBA) {materialFunction=&MF_FX_EmissiveAdjustments}
 ]
}
